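# Niagara effect for the TourBus skin cosmetic, apparently an extend moment, firing a short orange flare with sparks and debris.
# CoreFlare bursts a shrinking, camera-offset sprite, while Particle_Emitter launches spinning meshes in a narrow cone, steered by random acceleration and point attraction.
# Squares (GPU) and SquaresLow (CPU, lower spawn rate) emit small orange sprites that drag, shrink and fade.
# A Mover emitter is present but disabled.
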NiagaraSystem /BRCosmetics/Athena/Effects/Particles/Skins/TourBus/NS_TourBus_Extend_
dependencies=null
emitters {
 CoreFlare {
  emitterUpdateScript [
   &EmitterState(Life Cycle Mode=::NewEnumerator1, Loop Behavior=::NewEnumerator1, Loop Delay=0.1, Loop Duration=0.5, UseLoopDelay=true)
   &SpawnBurst_Instantaneous(Use Spawn Probability=false, UseLoopCountLimit=false)
  ]
  particleSpawnScript [
   &InitializeParticle(Sprite Size=<65 65>, Sprite Size Mode=::NewEnumerator4, UsePositionOffset=false, Write Material Random=true)
   &Color(Color=&MakeLinearColorFromVectorAndFloat("Vector (RGB)"=<1 0.5824 0>, "Float (Alpha)"=1))
  ]
  particleUpdateScript [
   &ParticleState()
   &ScaleSpriteSize(
    Scale Sprite Size Mode=::NewEnumerator0
    Uniform Scale Factor=&FloatFromCurve(FloatCurve=DataInterface() {propertyValues {Curve {keys [<at 0 1 RCIM_Cubic>,<at 1 0 RCIM_Cubic>]}}})
   )
   &CameraOffset(Camera Offset Amount=50)
   &ScaleColor(ScaleA=true, ScaleRGB=true)
   &DynamicMaterialParameters(
    Index 0 Param 1=0.1
    Index 0 Param 2=1
    Index 0 Param 3=1
    Index 0 Param 4=1
    Module.Param0WriteEnabled=true
    Module.Param1WriteEnabled=true
    Module.Param2WriteEnabled=true
    Module.Param3WriteEnabled=true
   )
  ]
  rendererClasses [&NiagaraSpriteRendererProperties]
  renderers [{rendererClass=&NiagaraSpriteRendererProperties}]
 }
 Particle_Emitter {
  emitterUpdateScript [
   &EmitterState(Life Cycle Mode=::NewEnumerator1, Loop Behavior=::NewEnumerator1, Loop Duration=0.75, UseLoopDelay=true)
   &SpawnBurst_Instantaneous(Spawn Count=5, Use Spawn Probability=false, UseLoopCountLimit=false)
   &SpawnRate(SpawnRate=12, Use Spawn Probability=false)
  ]
  particleSpawnScript [
   /SetVariables(Particles.MeshRotationEuler:Vector3f=&RandomRangeVector(Minimum=<0 0 0>, Maximum=<1 0 0>, Evaluation Type=::NewEnumerator1)) {isSetParameters=true}
   /SetVariables(Particles.MeshOrientation=/CascadeToNiagaraConverter/NiagaraScripts/DynamicInputScripts/MakeQuatFromEuler(Euler:Vector3f=$Particles.MeshRotationEuler, Scale:NiagaraFloat=360)) {isSetParameters=true}
   /SetVariables(Particles.MeshRotationRateBase:Vector3f=&RandomRangeVector(Minimum=<0 0 0>, Maximum=<1 0 0>, Evaluation Type=::NewEnumerator1)) {isSetParameters=true}
   /SetVariables(Particles.Acceleration:Vector3f=&RandomRangeVector(Minimum=<0 -150 -150>, Maximum=<0 150 150>, Evaluation Type=::NewEnumerator1)) {isSetParameters=true}
   &InitializeParticle(
    Lifetime=&RandomRangeFloat(Minimum=0.325, Maximum=1.25, Evaluation Type=::NewEnumerator1)
    Mesh Scale=&VectorFromFloat(Value=&RandomRangeFloat(Minimum=0.15, Maximum=0.5, Evaluation Type=::NewEnumerator1))
    Mesh Scale Mode=::NewEnumerator4
    UsePositionOffset=false
    Write Material Random=true
   )
   &AddVelocity(
    Cone Angle=&RandomRangeFloat(Minimum=0, Maximum=0.15, Evaluation Type=::NewEnumerator1)
    Cone Angle Mode:ENiagara_AngleInput=::NewEnumerator1
    Module.UseConeFalloff=false
    Module.UseVelDistribution=false
    Override Seed=false
    Velocity Mode=::NewEnumerator2
    Velocity Speed=&RandomRangeFloat(Minimum=80, Maximum=15, Evaluation Type=::NewEnumerator1)
   )
  ]
  particleUpdateScript [
   /SetVariables(Particles.MeshRotationRate:Vector3f=$Particles.MeshRotationRateBase) {isSetParameters=true}
   &ParticleState()
   &ScaleMeshSize(Scale Factor=&VectorFromFloat(Value=&FloatFromCurve(FloatCurve=DataInterface() {propertyValues {Curve {keys [
          {arriveTangent=2.7842, interpMode=RCIM_Cubic, leaveTangent=2.7842, tangentMode=RCTM_User, time=0, value=0}
          {arriveTangent=-0.0028, interpMode=RCIM_Cubic, leaveTangent=-0.0028, time=0.1642, value=0.9988}
          <at 1 -0.0028>
         ]}}})))
   &DynamicMaterialParameters(
    Index 0 Param 1=50000
    Index 0 Param 2=1
    Index 0 Param 3=&FloatFromCurve(FloatCurve=DataInterface() {propertyValues {Curve {keys [<at 0 0>,<at 1 1>]}}})
    Index 0 Param 4=1
    Module.Param0WriteEnabled=true
    Module.Param1WriteEnabled=true
    Module.Param2WriteEnabled=true
    Module.Param3WriteEnabled=true
   )
   /CascadeToNiagaraConverter/NiagaraScripts/ModuleScripts/UpdateMeshOrientationByEuler(RotationRate:Vector3f=$Particles.MeshRotationRate)
   &AccelerationForce(Acceleration=&RandomRangeVector(Minimum=<0 -150 -150>, Maximum=<0 150 150>))
   /Niagara/Modules/Update/Forces/PointAttractionForce(
    Attraction Radius=50
    AttractionStrength=80
    AttractorPosition:Vector3f=<500 0 0>
    Falloff Exponent=0.1
    Module.Kill Within Radius=false
    Module.Use Falloff=true
   )
   &SolveForcesAndVelocity(Clamp Velocity=false, Limit Acceleration=false)
  ]
  rendererClasses [&NiagaraMeshRendererProperties]
  renderers [{rendererClass=&NiagaraMeshRendererProperties, rendererIndex=0}]
 }
 Squares {
  simTarget=GPUComputeSim
  emitterUpdateScript [
   &EmitterState(Life Cycle Mode=::NewEnumerator1, Loop Behavior=::NewEnumerator1, Loop Delay=0.1, Loop Duration=0.75, UseLoopDelay=true)
   &SpawnBurst_Instantaneous(Spawn Count=12, Use Spawn Probability=false, UseLoopCountLimit=false)
   &SpawnRate(
    SpawnRate=&FloatFromCurve(FloatCurve=DataInterface() {propertyValues {Curve {keys [<at 0 50>,<at 1 25>]}}}, CurveIndex=$Emitter.NormalizedLoopAge)
    Use Spawn Probability=false
   )
  ]
  particleSpawnScript [
   &InitializeParticle(
    Lifetime=&RandomRangeFloat(Minimum=0.15, Maximum=0.5, Evaluation Type=::NewEnumerator1)
    Sprite Size=&RandomRangeVector2D(Minimum=<0.5 0.85>, Maximum=<1.85 2>, Evaluation Type=::NewEnumerator1)
    Sprite Size Mode=::NewEnumerator4
    UsePositionOffset=false
    Write Material Random=true
   )
   &Color(Color=&MakeLinearColorFromVectorAndFloat("Vector (RGB)"=<1 1 1>, "Float (Alpha)"=1))
   /CascadeToNiagaraConverter/NiagaraScripts/ModuleScripts/AddVectorToPosition(Vector:Vector3f=&TransformVector(Vector=<0 0 0>, Destination Space=::Simulation))
   &AddVelocity(
    Cone Angle=&RandomRangeFloat(Minimum=0, Maximum=0.15, Evaluation Type=::NewEnumerator1)
    Cone Angle Mode:ENiagara_AngleInput=::NewEnumerator1
    Module.UseConeFalloff=false
    Module.UseVelDistribution=false
    Override Seed=false
    Velocity Mode=::NewEnumerator2
    Velocity Speed=&RandomRangeFloat(Minimum=90, Maximum=22, Evaluation Type=::NewEnumerator1)
   )
  ]
  particleUpdateScript [
   &ParticleState()
   &ScaleSpriteSize(
    Scale Factor=&Vector2DFromFloat(Value=&FloatFromCurve(FloatCurve=DataInterface() {propertyValues {Curve {keys [<at 0 1 RCIM_Cubic>, {arriveTangent=-3.4115, interpMode=RCIM_Cubic, leaveTangent=-3.4115, tangentMode=RCTM_User, time=1, value=0}]}}}))
    Scale Sprite Size Mode=::NewEnumerator2
   )
   &Drag(Drag=0.1, Use Linear Drag=true, Use Rotational Drag=true)
   &SolveForcesAndVelocity(Clamp Velocity=false, Limit Acceleration=false)
   &ScaleColor(
    Scale Alpha=&FloatFromCurve(FloatCurve=DataInterface() {propertyValues {Curve {keys [
         <at 0 0 RCIM_Cubic>
         {arriveTangent=1.25, interpMode=RCIM_Cubic, leaveTangent=1.25, time=0.25, value=1}
         {arriveTangent=-1.3333, interpMode=RCIM_Cubic, leaveTangent=-1.3333, time=0.8, value=1}
         <at 1 0 RCIM_Cubic>
        ]}}})
    Scale RGB=<2 0.444 0.0633>
    ScaleA=true
    ScaleRGB=true
   )
  ]
  rendererClasses [&NiagaraSpriteRendererProperties]
  renderers [{rendererClass=&NiagaraSpriteRendererProperties}]
 }
 SquaresLow {
  emitterUpdateScript [
   &EmitterState(Life Cycle Mode=::NewEnumerator1, Loop Behavior=::NewEnumerator1, Loop Delay=0.1, Loop Duration=0.75, UseLoopDelay=true)
   &SpawnBurst_Instantaneous(Spawn Count=12, Use Spawn Probability=false, UseLoopCountLimit=false)
   &SpawnRate(
    SpawnRate=&FloatFromCurve(FloatCurve=DataInterface() {propertyValues {Curve {keys [<at 0 30>,<at 1 15>]}}}, CurveIndex=$Emitter.NormalizedLoopAge)
    Use Spawn Probability=false
   )
  ]
  particleSpawnScript [
   &InitializeParticle(
    Lifetime=&RandomRangeFloat(Minimum=0.15, Maximum=0.5, Evaluation Type=::NewEnumerator1)
    Sprite Size=&RandomRangeVector2D(Minimum=<0.5 0.85>, Maximum=<1.85 2>, Evaluation Type=::NewEnumerator1)
    Sprite Size Mode=::NewEnumerator4
    UsePositionOffset=false
    Write Material Random=true
   )
   &Color(Color=&MakeLinearColorFromVectorAndFloat("Vector (RGB)"=<1 1 1>, "Float (Alpha)"=1))
   &AddVectorToPosition(Vector:Vector3f=&TransformVector(Vector=<0 0 0>, Destination Space=::Simulation))
   &AddVelocity(
    Cone Angle=&RandomRangeFloat(Minimum=0, Maximum=0.15, Evaluation Type=::NewEnumerator1)
    Cone Angle Mode:ENiagara_AngleInput=::NewEnumerator1
    Module.UseConeFalloff=false
    Module.UseVelDistribution=false
    Override Seed=false
    Velocity Mode=::NewEnumerator2
    Velocity Speed=&RandomRangeFloat(Minimum=90, Maximum=22, Evaluation Type=::NewEnumerator1)
   )
  ]
  particleUpdateScript [
   &ParticleState()
   &ScaleSpriteSize(
    Scale Factor=&Vector2DFromFloat(Value=&FloatFromCurve(FloatCurve=DataInterface() {propertyValues {Curve {keys [<at 0 1 RCIM_Cubic>, {arriveTangent=-3.4115, interpMode=RCIM_Cubic, leaveTangent=-3.4115, tangentMode=RCTM_User, time=1, value=0}]}}}))
    Scale Sprite Size Mode=::NewEnumerator2
   )
   &Drag(Drag=0.1, Use Linear Drag=true, Use Rotational Drag=true)
   &SolveForcesAndVelocity(Clamp Velocity=false, Limit Acceleration=false)
   &ScaleColor(
    Scale Alpha=&FloatFromCurve(FloatCurve=DataInterface() {propertyValues {Curve {keys [
         <at 0 0 RCIM_Cubic>
         {arriveTangent=1.25, interpMode=RCIM_Cubic, leaveTangent=1.25, time=0.25, value=1}
         {arriveTangent=-1.3333, interpMode=RCIM_Cubic, leaveTangent=-1.3333, time=0.8, value=1}
         <at 1 0 RCIM_Cubic>
        ]}}})
    Scale RGB=<2 0.444 0.0633>
    ScaleA=true
    ScaleRGB=true
   )
  ]
  rendererClasses [&NiagaraSpriteRendererProperties]
  renderers [{rendererClass=&NiagaraSpriteRendererProperties}]
 }
}
summary {
 systemName=NS_TourBus_Extend_
 emitters [
  {emitterName=Particle_Emitter, rendererClasses [&NiagaraMeshRendererProperties]}
  {emitterName=Squares, simTarget=GPUComputeSim, rendererClasses [&NiagaraSpriteRendererProperties]}
  {emitterName=SquaresLow, rendererClasses [&NiagaraSpriteRendererProperties]}
  {emitterName=CoreFlare, rendererClasses [&NiagaraSpriteRendererProperties]}
  {bEnabled=false, emitterName=Mover, rendererClasses [&NiagaraSpriteRendererProperties]}
 ]
 userVariables []
}
